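# Translucent, two-sided unlit FX material for Dark Pit's eyes, drawing animated flame-like wisps.
# Panning swirl noise distorts two sphere masks that, remapped through curve atlas rows, shape opacity with a ball-joint mask and depth fade.
# Emissive blends distorted nebula texture, tinted by base and inside colors, with a curve-remapped edge color.
# World position offset adds SplineThicken widening to optional scrolling vector-noise curl, weighted by vertex blue.
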Material /BRCosmetics/Characters/Player/Male/Medium/Bodies/M_MED_Dark_Pit/FX/M_DarkPit_Eyes
graph {
 properties {
  bUsedWithMeshDeformer=true
  bUsedWithNiagaraMeshParticles=true
  bUsedWithSkeletalMesh=true
  bUsedWithSplineMeshes=true
  blendMode=BLEND_Translucent
  pixelDepthOffsetMode=PDOM_Legacy
  shadingModel=MSM_Unlit
  twoSided=true
  usedShadingModels=Unlit
 }
 nodes [
  n1 = LinearInterpolate(Alpha=VertexColor().G) {constA=1, constB=2}
  n2 = MaterialFunctionCall(
   WidthBase=Multiply(A=ScalarParameter() {defaultValue=1.5, parameterName=WidthBase}, B=@n1)
   WidthTip=Multiply(A=@n1, B=ScalarParameter() {defaultValue=18, parameterName=WidthTip})
  ) {materialFunction=&SplineThicken}
  n3 = VertexColor()
  n4 = WorldPosition()
  n5 = ActorPositionWS()
  n6 = ScalarParameter() {defaultValue=50, parameterName=Curl_Speed}
  n7 = Add(A=Time(), B=ScalarParameter() {parameterName=WPOTimeOffset})
  n8 = Divide(A=Subtract(A=Subtract(A=@n4.XYZ, B=@n5), B=Multiply(A=@n6, B=@n7)), B=ScalarParameter() {defaultValue=80, parameterName=Curl_Scale})
  n9 = ShadingPathSwitch(
   Default=Multiply(A=ScalarParameter() {defaultValue=3, parameterName=Curl_Intensity}, B=VectorNoise(World Position=@n8) {noiseFunction=VNF_VectorALU})
   Mobile=Constant() {r=1}
  ) {desc=Migrated from FeatureLevelSwitch}
  Add(
   A=@n2.WorldPositionOffset
   B=StaticSwitchParameter(True=Multiply(A=@n3.B, B=@n9), False=Constant() {r=1}) {defaultValue=true, parameterName=USE_WPO}
  )
  n10 = Panner() {speedX=0.05, speedY=0.3}
  n11 = Multiply(A=ComponentMask(None=@n10) {r=true}) {constB=0.3}
  n12 = Multiply(A=ComponentMask(None=@n10) {g=true}) {constB=0.4}
  n13 = Multiply(A=Panner() {speedX=0.03, speedY=0.1}) {constB=0.4}
  n14 = TextureObject() {texture=/Game/Characters/Player/Male/Medium/Bodies/M_Med_Soldier_04/Skins/Galaxy/Textures/T_FN_Nebula}
  n15 = TextureSample(
   UVs=LinearInterpolate(
    A=MaterialFunctionCall(X=@n11, Y=@n12) {materialFunction=&MakeFloat2}.Result
    B=TextureSample(UVs=@n13) {samplerType=SAMPLERTYPE_Normal, texture=/Game/Global/Textures/Default/Noise/T_SwirlNoise_N}.R
    Alpha=ScalarParameter() {defaultValue=0.07, parameterName=Distort_BlueFlame}
   )
   Tex=@n14
  )
  Desaturation(None=Multiply(A=@n15.RGB) {constB=2}, Fraction=Constant() {r=-1})
  n16 = VectorParameter() {parameterName=BaseColor, defaultValue=<rgba 0.1413 0.491 1 1>}
  n17 = LinearInterpolate(Alpha=ComponentMask(None=@n15.RGB) {r=true}) {constA=0.25}
  n18 = Multiply(A=Add(A=@n15.RGB), B=ComponentMask(None=@n15.RGB) {b=true})
  n19 = VectorParameter() {parameterName=InsideColor, defaultValue=<rgba 0.0076 2.7859 4 1>}
  n20 = Add(A=ComponentMask(None=TextureCoordinate()) {g=true}, B=ScalarParameter() {parameterName=Offset_color_diffuse}) {constB=0}
  n21 = Multiply(
   A=Add(A=Multiply(A=@n16.RGB, B=@n17), B=Multiply(A=@n18, B=@n19.RGB))
   B=LinearInterpolate(Alpha=Power(Base=@n20) {constExponent=4}) {constB=1.5}
  )
  n22 = Add(A=ComponentMask(None=TextureCoordinate()) {g=true}, B=ScalarParameter() {defaultValue=0.2, parameterName=Offset_color_Fresnel}) {constB=0}
  n23 = ScalarParameter() {defaultValue=0.7968, parameterName=Taper1}
  n24 = ComponentMask(None=TextureCoordinate()) {r=true}
  n25 = ScalarParameter() {defaultValue=2.2351, parameterName="Taper1_positon(UbaseV,0-1)"}
  n26 = ComponentMask(None=TextureCoordinate()) {g=true}
  n27 = Multiply(A=Divide(A=Constant() {r=1}, B=@n25), B=@n26)
  n28 = Constant() {r=1}
  n29 = Subtract(A=@n28, B=@n25)
  n30 = If(A=@n27, "A > B"=LinearInterpolate(A=Divide(A=@n28, B=@n29), Alpha=@n26) {constB=0}, "A == B"=@n27, "A < B"=@n27) {constB=1}
  n31 = ScalarParameter() {defaultValue=2.4226, parameterName=Taper1_power}
  n32 = LinearInterpolate(
   A=LinearInterpolate(A=Multiply(A=@n23) {constB=-1}, B=Add(A=@n23), Alpha=@n24) {constA=-0.1, constB=1.1}
   B=@n24
   Alpha=OneMinus(None=Power(Base=@n30, Exp=@n31) {constExponent=4})
  )
  n33 = Multiply(A=Time(), B=ScalarParameter() {defaultValue=5, parameterName=PanSpeed})
  n34 = Panner(
   Time=@n33
   Speed=AppendVector(
    A=ScalarParameter() {defaultValue=0.02, parameterName=DistortionSpeedU}
    B=ScalarParameter() {defaultValue=0.05, parameterName=DistortionSpeedV}
   )
  ) {speedX=0.1, speedY=0.1}
  n35 = ComponentMask(None=@n34) {r=true}
  n36 = ComponentMask(None=@n34) {g=true}
  n37 = ScalarParameter() {parameterName=MotionOffset}
  n38 = Add(A=AppendVector(A=Multiply(A=@n35) {constB=1.3}, B=Multiply(A=@n36) {constB=1.6}), B=@n37)
  n39 = ScalarParameter() {defaultValue=0.65, group=Flame_Area_Ctrl, parameterName=V_tile1}
  n40 = TextureCoordinate()
  n41 = Add(
   A=Multiply(
    A=TextureSample(UVs=@n38) {samplerType=SAMPLERTYPE_Normal, texture=&T_SwirlNoise_N}.R
    B=ScalarParameter() {defaultValue=0.453, parameterName=Distort1}
   ) {constB=0.3}
   B=LinearInterpolate(B=@n39, Alpha=ComponentMask(None=@n40) {g=true})
  )
  n42 = Constant() {r=0.5}
  n43 = Add(A=@n39, B=ScalarParameter() {defaultValue=0.06, parameterName=Sphere1_offset})
  n44 = ScalarParameter() {defaultValue=0.01, parameterName=DistortionSpeedU2}
  n45 = ScalarParameter() {defaultValue=0.1, parameterName=DistortionSpeedV2}
  n46 = Add(A=@n37, B=Panner(Time=@n33, Speed=AppendVector(A=@n44, B=@n45)) {speedX=0.1, speedY=0.1})
  n47 = ComponentMask(None=@n46) {r=true}
  n48 = ComponentMask(None=@n46) {g=true}
  n49 = TextureSample(UVs=AppendVector(A=Multiply(A=@n47) {constB=0.3}, B=Multiply(A=@n48) {constB=1.7})) {samplerType=SAMPLERTYPE_Normal, texture=&T_SwirlNoise_N}
  n50 = ScalarParameter() {defaultValue=1.2, parameterName=Distort2}
  n51 = ScalarParameter() {defaultValue=0.7704, group=Flame_Area_Ctrl, parameterName=V_tile2}
  n52 = ComponentMask(None=TextureCoordinate()) {g=true}
  n53 = AppendVector(
   A=ComponentMask(None=TextureCoordinate()) {r=true}
   B=Add(A=Multiply(A=@n49.R, B=@n50) {constB=0.3}, B=LinearInterpolate(B=@n51, Alpha=@n52))
  )
  n54 = AppendVector(A=Constant() {r=0.5}, B=Add(A=@n51, B=ScalarParameter() {defaultValue=-0.4546, parameterName=Sphere2_offset}) {constB=-0.8522})
  n55 = ScalarParameter() {defaultValue=0.3, group=Flame_Area_Ctrl, parameterName=Radius2, sortPriority=1}
  n56 = ScalarParameter() {defaultValue=0.2, group=Flame_Area_Ctrl, parameterName=Hardness2, sortPriority=2}
  n57 = Add(
   A=SphereMask(
    A=AppendVector(A=@n32, B=@n41)
    B=AppendVector(A=@n42, B=@n43)
    Radius=ScalarParameter() {defaultValue=0.6, group=Flame_Area_Ctrl, parameterName=Radius1, sortPriority=1}
    Hardness=ScalarParameter() {defaultValue=0.3, group=Flame_Area_Ctrl, parameterName=Hardness1, sortPriority=2}
   ) {attenuationRadius=1.5, hardnessPercent=0}
   B=StaticSwitchParameter(True=SphereMask(A=@n53, B=@n54, Radius=@n55, Hardness=@n56) {attenuationRadius=1.5, hardnessPercent=0}, False=Constant() {r=1}) {defaultValue=true, parameterName=Use_secondSphere}
  ) {constB=0}
  n58 = CurveAtlasRowParameter(CurveTime=@n57) {
   atlas=&AthenaCurveAtlas01
   curve=/Game/Characters/Player/Male/Medium/Bodies/M_MED_VoyagerRemix/FX/Formal/Curve_VoyagerRemix2
   defaultValue=119
  }
  n59 = Multiply(
   A=LinearInterpolate(Alpha=Power(Base=@n22) {constExponent=6}) {constB=1.5}
   B=Multiply(A=VectorParameter() {parameterName=EdgeColor, defaultValue=<rgba 0.0422 0.3542 0.5 1>}.RGB, B=@n58.B)
  )
  MaterialFunctionCall(Emissive=Multiply(A=Add(A=@n59, B=@n21), B=ScalarParameter() {defaultValue=2, parameterName=Emissive_intensity})) {materialFunction=&MF_FX_EmissiveAdjustments}
  n60 = ScalarParameter() {defaultValue=5, parameterName=DistTileMult}
  n61 = MaterialFunctionCall(Distance=ScalarParameter() {defaultValue=800, parameterName=TileFadeDistance}) {materialFunction=&Near_Camera_Fade}
  MaterialFunctionCall!(
   CameraVector
   Rotation Speed=Multiply(A=Time(), B=ScalarParameter() {defaultValue=0.05, group=Galaxy, parameterName=Nebula Rotation speed})
   Texture=@n14
   Tiling=Multiply(A=ScalarParameter() {defaultValue=1, parameterName=Nebula_Base_tile}, B=LinearInterpolate(B=@n60, Alpha=@n61.Result) {constA=1})
   Offset
   AxisFadeContrast
  ) {materialFunction=/Game/Weapons/WeaponSkins/MaterialFunction/TriplanarCameraVector_Rotated}
  n62 = Panner(Time=@n33) {speedY=0.1}
  n63 = Constant2Vector() {g=1, r=1}
  n64 = TextureCoordinate()
  n65 = CurveAtlasRowParameter(CurveTime=OneMinus(None=ComponentMask(None=@n64) {g=true})) {
   atlas=&AthenaCurveAtlas01
   curve=&Curve_VoyagerRemix2
   defaultValue=119
  }
  n66 = LinearInterpolate(
   A=TextureSample(UVs=Multiply(A=@n62, B=@n63)) {samplerType=SAMPLERTYPE_Normal, texture=&T_SwirlNoise_N}.R
   Alpha=@n65.G
  )
  n67 = TextureCoordinate()
  n68 = Multiply(
   A=@n67
   B=AppendVector(A=Constant() {r=1}, B=ScalarParameter() {defaultValue=1, group=BallJoint Mask, parameterName=BallJoint_Sphere-Squish})
  )
  n69 = AppendVector(A=Constant() {r=0.5}, B=ScalarParameter() {defaultValue=0.55, group=BallJoint Mask, parameterName=BallJoint_Sphere-Postiion})
  n70 = ScalarParameter() {defaultValue=0.5, group=BallJoint Mask, parameterName=BallJoint_Radius}
  n71 = ScalarParameter() {defaultValue=0.3, group=BallJoint Mask, parameterName=BallJoint_Hardness}
  n72 = ComponentMask(None=@n67) {g=true}
  n73 = Clamp(None=Multiply(A=@n72, B=ScalarParameter() {defaultValue=2, group=BallJoint Mask, parameterName=BallJointGrad_Tiling}))
  n74 = ScalarParameter() {defaultValue=5, group=BallJoint Mask, parameterName=BallJoint_GradPower}
  n75 = Add(
   A=Add(A=SphereMask(A=@n68, B=@n69, Radius=@n70, Hardness=@n71), B=Power(Base=@n73, Exp=@n74))
   B=OneMinus(None=ScalarParameter() {defaultValue=1, group=BallJoint Mask, parameterName=BallJoint_Mask?})
  )
  n76 = CurveAtlasRowParameter(CurveTime=Multiply(A=Multiply(A=@n66, B=@n57), B=Clamp(None=@n75))) {
   atlas=&AthenaCurveAtlas01
   curve=&Curve_VoyagerRemix2
   defaultValue=119
  }
  n77 = Multiply(A=@n65.G, B=@n76.R) {constA=1}
  n78 = DepthFade(Opacity=@n77, FadeDistance=ScalarParameter() {defaultValue=2, parameterName=DepthFade}) {fadeDistanceDefault=2}
  Multiply(
   A=ShadingPathSwitch(Default=@n78, Deferred=@n78, Mobile=@n77) {desc=Migrated from FeatureLevelSwitch}
   B=ScalarParameter() {defaultValue=1.2, parameterName=OverallOpacity}
  )
  OneMinus(None=@n72)
  Reroute(None=@n2.WorldPositionOffset)
  ShadingPathSwitch(Default=@n15.RGB, Deferred=@n15.RGB, Mobile=@n15.RGB) {desc=Migrated from FeatureLevelSwitch}
  VectorParameter() {parameterName=sphere_bottom_texture, defaultValue=<rgba 0.1 0.9 1 1>}
 ]
}
