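# Niagara system for the Dark Pit character's eye effects: cyan sprite flames, glows, flickering lights, drifting goop meshes and an eye trail mesh.
# Flames, glows and lights follow eye sockets on User.SkeletalMesh; two goop emitters spawn at the next two socket indices.
# EyeTrail uses a fixed box offset, with its socket sampling disabled.
# User.bisHeadHidden sets particle visibility tags and switches off the lights, whose flicker follows loop-age curves.
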NiagaraSystem /BRCosmetics/Characters/Player/Male/Medium/Bodies/M_MED_Dark_Pit/FX/NS_M_DarkPit_Eyes
dependencies {
 usedDataInterfaces [
  &NiagaraDataInterfaceCurve
  &NiagaraDataInterfaceVectorCurve
  &NiagaraDataInterfaceColorCurve
  &NiagaraDataInterfaceArrayFloat3
 ]
 usedDynamicInputs [&VectorFromCurve]
 usedModules [&SystemState]
}
emitters {
 EyeFlamesLeft {
  emitterUpdateScript [
   &EmitterState()
   &SpawnRate(SpawnRate=24, Use Spawn Probability=false)
  ]
  particleSpawnScript [
   /Niagara/Modules/Spawn/Initialization/InitializeParticle(
    Color=<rgba 0.25 0.825 1 0.75>
    Lifetime=&UniformRangedFloat(Minimum=0.55, Maximum=0.65)
    Module.Write Color=true
    Module.Write Lifetime=true
    Module.Write Mass=false
    Module.Write Material Random=true
    Module.Write Position=false
    Module.Write Scale=false
    Module.Write SpriteRotation=true
    Module.Write SpriteSize=true
    Position=&SimulationPosition()
    Sprite Rotation=&UniformRangedFloat(Minimum=30, Maximum=120)
    Sprite Size=&Vector2DFromFloat(Value=&RandomRangeFloat(Minimum=6, Maximum=7))
   )
   &DynamicMaterialParameters(
    Index 0 Param 2=&UniformRangedFloat(Minimum=0, Maximum=1)
    Index 0 Param 3=1
    Index 0 Param 4=&UniformRangedFloat(Minimum=1, Maximum=1)
    Module.Param0WriteEnabled=false
    Module.Param1WriteEnabled=true
    Module.Param2WriteEnabled=false
    Module.Param3WriteEnabled=true
   ) {module=DynamicMaterialParameters002}
   &FN_SetSpawnGroup(Number of Spawn Groups=1)
   &SkeletalMeshLocation(
    Bone / Socket Index=&SpawnGroup()
    Inherit Velocity=false
    Interpolate Mesh Sampling=false
    MaskSpawnGroup=false
    Mesh Sampling Type=::NewEnumerator5
    Sampled Position Offset=<-1.5 1.5 -0.7>
    Sampled Position Offset Coordinate Space:ENiagaraMeshTransforms=::NewEnumerator3
    Skeletal Mesh=$User.SkeletalMesh
    Socket Sampling Mode=::NewEnumerator5
   )
   &CameraOffset(Camera Offset Amount=&RandomRangeFloat(Minimum=0.8, Maximum=1.5))
   &AddVelocity(Velocity=&RandomRangeVector(Minimum=<8 0 -3>, Maximum=<3 0 -6>))
  ]
  particleUpdateScript [
   &ParticleState()
   &ScaleColor(
    Scale Mode=::NewEnumerator1
    Scale RGBA=&Vector4FromCurve(Vector4Curve=DataInterface() {propertyValues {
       AlphaCurve {keys [<at 0.0024 0>,<at 0.1965 2>,<at 0.5 3>,<at 1 0>]}
       BlueCurve {keys [<at 0.2365 4.0833>,<at 0.65 1>,<at 0.85 1>]}
       GreenCurve {keys [<at 0.2365 5>,<at 0.65 0.2456>,<at 0.85 0.2333>]}
       RedCurve {keys [<at 0.2365 0>,<at 0.65 0.016>,<at 0.85 0>]}
      }})
    ScaleRGBA=true
   )
   /SetVariables(Particles.SubImageIndex=&FloatFromCurve(FloatCurve=DataInterface() {propertyValues {Curve {keys [<at 0 10>,<at 1 34>]}}})) {isSetParameters=true, module=SetVariables_#1}
   &ScaleSpriteSize(Scale Factor=&Vector2DFromFloat(Value=&FloatFromCurve(FloatCurve=DataInterface() {propertyValues {Curve {keys [
          {arriveTangent=3.75, interpMode=RCIM_Cubic, leaveTangent=3.75, time=0, value=0}
          {arriveTangent=3.8351, interpMode=RCIM_Cubic, leaveTangent=3.8351, tangentMode=RCTM_User, time=0.2, value=0.75}
          {arriveTangent=-0.125, interpMode=RCIM_Cubic, leaveTangent=-0.125, time=0.5752, value=1.25}
          {arriveTangent=-1.4124, interpMode=RCIM_Cubic, leaveTangent=-1.4124, time=1, value=0.65}
         ]}}})))
   &DynamicMaterialParameters(
    Index 0 Param 1=&FloatFromCurve(FloatCurve=DataInterface() {propertyValues {Curve {keys [{arriveTangent=2.7551, interpMode=RCIM_Cubic, leaveTangent=2.7551, tangentMode=RCTM_User, time=0, value=0}, <at 0.5 1>]}}})
    Index 0 Param 2=&RandomRangeFloat(Minimum=0.33, Maximum=1)
    Index 0 Param 3=&FloatFromCurve(FloatCurve=DataInterface() {propertyValues {Curve {keys [<at 0.8 1>,<at 1 2>]}}})
    Index 0 Param 4=1
    Module.Param0WriteEnabled=true
    Module.Param1WriteEnabled=true
    Module.Param2WriteEnabled=true
    Module.Param3WriteEnabled=true
   ) {module=DynamicMaterialParameters001}
   &SpriteRotationRate()
   &SolveForcesAndVelocity(Clamp Velocity=false, Limit Acceleration=false)
   /SetVariables(Particles.VisibilityTag=&MakeCustomIntFromBool(Bool=$User.bisHeadHidden, True INT=1, False INT=0)) {isSetParameters=true, module=SetVariables_#2}
  ]
  rendererClasses [&NiagaraSpriteRendererProperties]
  renderers [{rendererClass=&NiagaraSpriteRendererProperties}]
 }
 EyeFlamesRight {
  emitterUpdateScript [
   &EmitterState()
   &SpawnRate(SpawnRate=24, Use Spawn Probability=false)
  ]
  particleSpawnScript [
   &InitializeParticle(
    Color=<rgba 0.25 0.825 1 0.75>
    Lifetime=&UniformRangedFloat(Minimum=0.55, Maximum=0.65)
    Module.Write Color=true
    Module.Write Lifetime=true
    Module.Write Mass=false
    Module.Write Material Random=true
    Module.Write Position=false
    Module.Write Scale=false
    Module.Write SpriteRotation=true
    Module.Write SpriteSize=true
    Position=&SimulationPosition()
    Sprite Rotation=&UniformRangedFloat(Minimum=-30, Maximum=-120)
    Sprite Size=&Vector2DFromFloat(Value=&RandomRangeFloat(Minimum=6, Maximum=7))
   )
   &DynamicMaterialParameters(
    Index 0 Param 2=&UniformRangedFloat(Minimum=0, Maximum=1)
    Index 0 Param 3=1
    Index 0 Param 4=&UniformRangedFloat(Minimum=1, Maximum=1)
    Module.Param0WriteEnabled=false
    Module.Param1WriteEnabled=true
    Module.Param2WriteEnabled=false
    Module.Param3WriteEnabled=true
   ) {module=DynamicMaterialParameters002}
   &FN_SetSpawnGroup(Number of Spawn Groups=1) {module=FN_SetSpawnGroup001}
   &SkeletalMeshLocation(
    Bone / Socket Index=1
    Inherit Velocity=false
    Interpolate Mesh Sampling=false
    MaskSpawnGroup=false
    Mesh Sampling Type=::NewEnumerator5
    Sampled Position Offset=<-1.5 1.5 -1.5>
    Sampled Position Offset Coordinate Space:ENiagaraMeshTransforms=::NewEnumerator3
    Skeletal Mesh=$User.SkeletalMesh
    Socket Sampling Mode=::NewEnumerator5
   ) {module=SkeletalMeshLocation002}
   &CameraOffset(Camera Offset Amount=&RandomRangeFloat(Minimum=0.8, Maximum=1.5))
   &AddVelocity(Velocity=&RandomRangeVector(Minimum=<8 0 3>, Maximum=<3 0 6>))
  ]
  particleUpdateScript [
   &ParticleState()
   &ScaleColor(
    Scale Mode=::NewEnumerator1
    Scale RGBA=&Vector4FromCurve(Vector4Curve=DataInterface() {propertyValues {
       AlphaCurve {keys [<at 0.0024 0>,<at 0.1965 2>,<at 0.5 3>,<at 1 0>]}
       BlueCurve {keys [<at 0.2365 4.0833>,<at 0.65 1>,<at 0.85 1>]}
       GreenCurve {keys [<at 0.2365 5>,<at 0.65 0.2456>,<at 0.85 0.2333>]}
       RedCurve {keys [<at 0.2365 0>,<at 0.65 0.016>,<at 0.85 0>]}
      }})
    ScaleRGBA=true
   )
   /SetVariables(Particles.SubImageIndex=&FloatFromCurve(FloatCurve=DataInterface() {propertyValues {Curve {keys [<at 0 10>,<at 1 34>]}}})) {isSetParameters=true, module=SetVariables_#1}
   &ScaleSpriteSize(Scale Factor=&Vector2DFromFloat(Value=&FloatFromCurve(FloatCurve=DataInterface() {propertyValues {Curve {keys [
          {arriveTangent=3.75, interpMode=RCIM_Cubic, leaveTangent=3.75, time=0, value=0}
          {arriveTangent=3.8351, interpMode=RCIM_Cubic, leaveTangent=3.8351, tangentMode=RCTM_User, time=0.2, value=0.75}
          {arriveTangent=-0.125, interpMode=RCIM_Cubic, leaveTangent=-0.125, time=0.5752, value=1.25}
          {arriveTangent=-1.4124, interpMode=RCIM_Cubic, leaveTangent=-1.4124, time=1, value=0.65}
         ]}}})))
   &DynamicMaterialParameters(
    Index 0 Param 1=&FloatFromCurve(FloatCurve=DataInterface() {propertyValues {Curve {keys [{arriveTangent=2.7551, interpMode=RCIM_Cubic, leaveTangent=2.7551, tangentMode=RCTM_User, time=0, value=0}, <at 0.5 1>]}}})
    Index 0 Param 2=&RandomRangeFloat(Minimum=0.33, Maximum=1)
    Index 0 Param 3=&FloatFromCurve(FloatCurve=DataInterface() {propertyValues {Curve {keys [<at 0.8 1>,<at 1 2>]}}})
    Index 0 Param 4=1
    Module.Param0WriteEnabled=true
    Module.Param1WriteEnabled=true
    Module.Param2WriteEnabled=true
    Module.Param3WriteEnabled=true
   ) {module=DynamicMaterialParameters001}
   &SpriteRotationRate()
   &SolveForcesAndVelocity(Clamp Velocity=false, Limit Acceleration=false)
   /SetVariables(Particles.VisibilityTag=&MakeCustomIntFromBool(Bool=$User.bisHeadHidden, True INT=1, False INT=0)) {isSetParameters=true, module=SetVariables_#2}
  ]
  rendererClasses [&NiagaraSpriteRendererProperties]
  renderers [{rendererClass=&NiagaraSpriteRendererProperties}]
 }
 EyeTrail {
  emitterUpdateScript [
   &EmitterState(Life Cycle Mode=::NewEnumerator1, Loop Behavior=::NewEnumerator1, Loop Duration=1, UseLoopDelay=false)
   &SpawnBurst_Instantaneous(Use Spawn Probability=false, UseLoopCountLimit=false)
  ]
  particleSpawnScript [
   /Niagara/Modules/Spawn/Initialization/V2/InitializeParticle(
    Color=<rgba 1 1 1 3>
    Lifetime Max=2.5
    Lifetime Min=1.5
    Lifetime Mode=::NewEnumerator1
    Mesh Scale Mode=::NewEnumerator3
    UsePositionOffset=false
    Write Material Random=true
   )
   &InitialMeshOrientation(Orientation Vector=<0 0 -1>, Use Orientation Vector=true, Use Rotation Vector=false)
   &BoxLocation(Box Offset=<-152 3 0>, Box Size=<0.1 0.1 0.1>, MaskSpawnGroup=false, Module.Override Seed=false, Rotate Box=false, Surface Only=false)
  ]
  particleUpdateScript [
   &ParticleState(Kill Particles When Lifetime Has Elapsed=false)
   &DynamicMaterialParameters(
    Index 0 Param 1=1
    Index 0 Param 3=1
    Module.Param0WriteEnabled=true
    Module.Param1WriteEnabled=true
    Module.Param2WriteEnabled=true
    Module.Param3WriteEnabled=true
   )
   &CameraOffset(Camera Offset Amount=5) {enabled=false}
   &SkeletalMeshLocation(
    MaskSpawnGroup=false
    Mesh Sampling Type=::NewEnumerator5
    Skeletal Mesh=DataInterface() {propertyValues {
      DefaultMesh=null
      ExcludeBoneName=root
      OptionalComponentClass=null
      PreviewMesh=/BRCosmetics/Characters/Player/Male/Medium/Bodies/M_MED_Vivid_Trooper/Meshes/M_MED_Vivid_Trooper
      SkinningMode=SkinOnTheFly
      SoftSourceActor=null
      SourceMode=Default
      UvSetIndex=0
      WholeMeshLOD=-1
      bExcludeBone=false
      bReadDeformedGeometry=true
      bRequireCurrentFrameData=true
      ComponentTags []
      FilteredBones [head]
      FilteredSockets [FX_Headsocket]
      MeshUserParameter {parameter {name=None, typeDefHandle {registeredTypeIndex=1}, varData []}}
      SamplingRegions []
     }}
    Socket Sampling Mode=::NewEnumerator5
   ) {enabled=false}
   &SolveForcesAndVelocity(Clamp Velocity=false, Limit Acceleration=false) {enabled=false}
   &SkeletalMeshLocation(
    MaskSpawnGroup=false
    Mesh Sampling Type=::NewEnumerator5
    Skeletal Mesh=DataInterface() {propertyValues {
      DefaultMesh=null
      ExcludeBoneName=root
      OptionalComponentClass=null
      PreviewMesh=&M_MED_Vivid_Trooper
      SkinningMode=SkinOnTheFly
      SoftSourceActor=null
      SourceMode=Default
      UvSetIndex=0
      WholeMeshLOD=-1
      bExcludeBone=false
      bReadDeformedGeometry=true
      bRequireCurrentFrameData=true
      ComponentTags []
      FilteredBones []
      FilteredSockets [FX_Calf_l]
      MeshUserParameter {parameter {name=None, typeDefHandle {registeredTypeIndex=1}, varData []}}
      SamplingRegions []
     }}
    Socket Sampling Mode=::NewEnumerator5
   ) {enabled=false, module=SkeletalMeshLocation001}
   /SetVariables(Particles.VisibilityTag=&MakeCustomIntFromBool(Bool=$User.bisHeadHidden, True INT=1, False INT=0)) {isSetParameters=true, module=SetVariables_#2}
  ]
  rendererClasses [&NiagaraMeshRendererProperties]
  renderers [
   {rendererClass=&NiagaraMeshRendererProperties, rendererIndex=0}
   {rendererClass=&NiagaraMeshRendererProperties, rendererIndex=1}
  ]
 }
 Glow {
  emitterUpdateScript [
   &EmitterState()
   &SpawnRate(SpawnRate=24, Use Spawn Probability=false)
  ]
  particleSpawnScript [
   /Niagara/Modules/Spawn/Initialization/InitializeParticle(
    Color=<rgba 0 1.2267 2 1>
    Lifetime=&UniformRangedFloat(Minimum=0.45, Maximum=0.5)
    Module.Write Color=true
    Module.Write Lifetime=true
    Module.Write Mass=true
    Module.Write Material Random=true
    Module.Write Position=true
    Module.Write Scale=false
    Module.Write SpriteRotation=false
    Module.Write SpriteSize=true
    Position=&SimulationPosition()
    Sprite Size=&Vector2DFromFloat(Value=&UniformRangedFloat(Minimum=9, Maximum=9))
   )
   &DynamicMaterialParameters(
    Index 0 Param 1=&RandomRangeFloat(Minimum=0, Maximum=1)
    Module.Param0WriteEnabled=true
    Module.Param1WriteEnabled=false
    Module.Param2WriteEnabled=false
    Module.Param3WriteEnabled=false
   )
   &FN_SetSpawnGroup(Number of Spawn Groups=2)
   &CameraOffset(Camera Offset Amount=2.5)
  ]
  particleUpdateScript [
   &ParticleState()
   &SkeletalMeshLocation(
    Bone / Socket Index=&SpawnGroup()
    Inherit Velocity=false
    Interpolate Mesh Sampling=false
    MaskSpawnGroup=false
    Mesh Sampling Type=::NewEnumerator5
    Sampled Position Offset=&MaskVectorBySpawnGroup(Masked Vector=<0.4 0.5 -0.5>, Passthrough Vector=<-0.4 0.5 -0.5>)
    Sampled Position Offset Coordinate Space:ENiagaraMeshTransforms=::NewEnumerator3
    Skeletal Mesh=$User.SkeletalMesh
    Socket Sampling Mode=::NewEnumerator5
   ) {module=SkeletalMeshLocation001}
   &ScaleColor(
    Scale Alpha=&FloatFromCurve(FloatCurve=DataInterface() {propertyValues {Curve {keys [<at 0 0>,<at 0.25 1>,<at 0.8107 1>,<at 1 0>]}}})
    ScaleA=true
    ScaleRGB=true
   )
   &ScaleSpriteSize(Scale Factor=&Vector2DFromFloat(Value=&FloatFromCurve(FloatCurve=DataInterface() {propertyValues {Curve {keys [
          {arriveTangent=-0.0036, interpMode=RCIM_Cubic, leaveTangent=-0.0036, tangentMode=RCTM_User, time=0, value=1}
          {arriveTangent=-2.2201, interpMode=RCIM_Cubic, leaveTangent=-2.2201, tangentMode=RCTM_User, time=1, value=0.1}
         ]}}})))
   /SetVariables(Particles.VisibilityTag=&MakeCustomIntFromBool(Bool=$User.bisHeadHidden, True INT=1, False INT=0)) {isSetParameters=true, module=SetVariables_#2}
  ]
  rendererClasses [&NiagaraSpriteRendererProperties]
  renderers [{rendererClass=&NiagaraSpriteRendererProperties}]
 }
 Goop {
  emitterUpdateScript [
   &EmitterState(Inactive Response=::NewEnumerator1, Life Cycle Mode=::NewEnumerator1, Loop Duration=2, UseLoopDelay=false)
   &SpawnRate(Spawn Probability=0.8, SpawnRate=7, Use Spawn Probability=true)
  ]
  particleSpawnScript [
   /Niagara/Modules/Spawn/Initialization/V2/InitializeParticle(
    Color=<rgba 0 0.8911 1 1>
    Lifetime Max=1.8
    Lifetime Min=1.5
    Lifetime Mode=::NewEnumerator1
    Mesh Scale Max=<0.6 0.6 1>
    Mesh Scale Min=<0.2 0.2 0.6>
    Mesh Scale Mode=::NewEnumerator2
    Position Offset=<0 0 -4>
    UsePositionOffset=true
    Write Material Random=true
   )
   &DynamicMaterialParameters(
    Index 0 Param 1=2
    Index 0 Param 2=10
    Index 0 Param 3=3
    Index 0 Param 4=&RandomRangeFloat(Minimum=50, Maximum=55, Evaluation Type=::NewEnumerator1)
    Module.Param0WriteEnabled=true
    Module.Param1WriteEnabled=true
    Module.Param2WriteEnabled=true
    Module.Param3WriteEnabled=false
   ) {module=DynamicMaterialParameters002}
   &BoxLocation(
    Box Offset=&SelectVectorFromArray(Direct Array Index=$Emitter.SpawnGroup, Vector Selection Array=DataInterface() {propertyValues {FloatData [<11 -6.1 5>, <-11 -6.1 5>]}})
    Box Size=<0 0 0>
    MaskSpawnGroup=false
    Module.Override Seed=false
    Rotate Box=false
    Surface Only=false
   ) {enabled=false}
   &AddVelocityFromPoint(Module.Use Falloff Distance=false, Origin Offset=<0 -5 2.5>, Velocity Strength=6) {enabled=false}
   &AddVelocity(Coordinate Space=::World) {enabled=false}
   &SkeletalMeshLocation(
    MaskSpawnGroup=false
    Mesh Sampling Type=::NewEnumerator5
    Skeletal Mesh=DataInterface() {propertyValues {
      DefaultMesh=null
      ExcludeBoneName=root
      OptionalComponentClass=null
      PreviewMesh=/BRCosmetics/Characters/Player/Male/Medium/Bodies/M_MED_Dark_Pit/Meshes/Parts/M_MED_Dark_Pit_FaceAcc
      SkinningMode=SkinOnTheFly
      SoftSourceActor=null
      SourceMode=Default
      UvSetIndex=0
      WholeMeshLOD=-1
      bExcludeBone=false
      bReadDeformedGeometry=true
      bRequireCurrentFrameData=true
      ComponentTags []
      FilteredBones []
      FilteredSockets [FX_Goop_L]
      MeshUserParameter {parameter {name=None, typeDefHandle {registeredTypeIndex=1}, varData []}}
      SamplingRegions []
     }}
    Socket Sampling Mode=::NewEnumerator5
   ) {enabled=false}
   &FN_SetSpawnGroup(Number of Spawn Groups=2)
   &SkeletalMeshLocation(
    Bone / Socket Index=&Add_Integer(A=&SpawnGroup(), B=2)
    Inherit Velocity=false
    Interpolate Mesh Sampling=false
    MaskSpawnGroup=false
    Mesh Sampling Type=::NewEnumerator5
    Sampled Position Offset=<-1.5 1.5 -0.7>
    Sampled Position Offset Coordinate Space:ENiagaraMeshTransforms=::NewEnumerator3
    Skeletal Mesh=$User.SkeletalMesh
    Socket Sampling Mode=::NewEnumerator5
   ) {module=SkeletalMeshLocation001}
   &SphereLocation(
    MaskSpawnGroup=false
    Override Seed=false
    Sphere Radius=1.5
    Surface Only=false
    UseNonUniformScale=false
    UseOffset=false
    UseSphereOrientation=false
   )
  ]
  particleUpdateScript [
   &ParticleState()
   &DynamicMaterialParameters(
    Index 0 Param 3=&FloatFromCurve(
     FloatCurve=DataInterface() {propertyValues {Curve {keys [
         <at 0 0 RCIM_Cubic>
         {arriveTangent=2.4962, interpMode=RCIM_Cubic, leaveTangent=2.4962, tangentMode=RCTM_User, time=0.3, value=0.45}
         <at 1 1>
        ]}}}
     Scale Curve=2
    )
    Module.Param0WriteEnabled=false
    Module.Param1WriteEnabled=false
    Module.Param2WriteEnabled=true
    Module.Param3WriteEnabled=false
   )
   &ScaleMeshSize(Scale Factor=&VectorFromFloat(Value=&FloatFromCurve(
      FloatCurve=DataInterface() {propertyValues {Curve {keys [
          {arriveTangent=-0.026, interpMode=RCIM_Cubic, leaveTangent=-0.026, tangentMode=RCTM_User, time=0, value=0}
          {arriveTangent=0.1004, interpMode=RCIM_Cubic, leaveTangent=0.1004, tangentMode=RCTM_User, time=0.6, value=0.7}
          {arriveTangent=-3.4821, interpMode=RCIM_Cubic, leaveTangent=-3.4821, tangentMode=RCTM_User, time=1, value=0}
         ]}}}
      Scale Curve=0.6
     ))) {module=ScaleMeshSize001}
   &GravityForce(Coordinate Space=::Local, Gravity=<16 -5 0>)
   &Drag(Use Linear Drag=true, Use Rotational Drag=false)
   &MeshRotationForce(
    Coordinate Space:ENiagaraMeshTransforms=::NewEnumerator3
    Rotation=&RandomRangeVector(Minimum=<-4 -4 -4>, Maximum=<4 4 4>)
   )
   &SolveForcesAndVelocity(Clamp Velocity=false, Limit Acceleration=false)
   &SolveRotationalForcesAndVelocity()
   &CameraOffset(Camera Offset Amount=&RandomRangeFloat(Minimum=0, Maximum=-3))
   /SetVariables(Particles.VisibilityTag=&MakeCustomIntFromBool(Bool=$User.bisHeadHidden, True INT=1, False INT=0)) {isSetParameters=true, module=SetVariables_#2}
  ]
  rendererClasses [&NiagaraMeshRendererProperties]
  renderers [
   {rendererClass=&NiagaraMeshRendererProperties, rendererIndex=0}
   {rendererClass=&NiagaraMeshRendererProperties, rendererIndex=1}
  ]
 }
 Goop_Big {
  emitterUpdateScript [
   &EmitterState(Inactive Response=::NewEnumerator1, Life Cycle Mode=::NewEnumerator1, Loop Duration=2, UseLoopDelay=false)
   &SpawnRate(Spawn Probability=0.8, SpawnRate=2, Use Spawn Probability=true)
  ]
  particleSpawnScript [
   &InitializeParticle(
    Color=<rgba 0 0.8911 1 1>
    Lifetime Max=1.8
    Lifetime Min=1.5
    Lifetime Mode=::NewEnumerator1
    Mesh Scale Max=<1.5 1.5 2>
    Mesh Scale Min=<0.5 0.5 1>
    Mesh Scale Mode=::NewEnumerator2
    Position Offset=<0 0 -4>
    UsePositionOffset=true
    Write Material Random=true
   )
   &DynamicMaterialParameters(
    Index 0 Param 1=2
    Index 0 Param 2=5
    Index 0 Param 3=3
    Index 0 Param 4=&RandomRangeFloat(Minimum=50, Maximum=55, Evaluation Type=::NewEnumerator1)
    Module.Param0WriteEnabled=true
    Module.Param1WriteEnabled=true
    Module.Param2WriteEnabled=true
    Module.Param3WriteEnabled=false
   ) {module=DynamicMaterialParameters002}
   &BoxLocation(
    Box Offset=&SelectVectorFromArray(Direct Array Index=$Emitter.SpawnGroup, Vector Selection Array=DataInterface() {propertyValues {FloatData [<11 -6.1 5>, <-11 -6.1 5>]}})
    Box Size=<0 0 0>
    MaskSpawnGroup=false
    Module.Override Seed=false
    Rotate Box=false
    Surface Only=false
   ) {enabled=false}
   &AddVelocityFromPoint(Module.Use Falloff Distance=false, Origin Offset=<0 -5 2.5>, Velocity Strength=6) {enabled=false}
   &AddVelocity(Coordinate Space=::World) {enabled=false}
   &SkeletalMeshLocation(
    MaskSpawnGroup=false
    Mesh Sampling Type=::NewEnumerator5
    Skeletal Mesh=DataInterface() {propertyValues {
      DefaultMesh=null
      ExcludeBoneName=root
      OptionalComponentClass=null
      PreviewMesh=&M_MED_Dark_Pit_FaceAcc
      SkinningMode=SkinOnTheFly
      SoftSourceActor=null
      SourceMode=Default
      UvSetIndex=0
      WholeMeshLOD=-1
      bExcludeBone=false
      bReadDeformedGeometry=true
      bRequireCurrentFrameData=true
      ComponentTags []
      FilteredBones []
      FilteredSockets [FX_Goop_L]
      MeshUserParameter {parameter {name=None, typeDefHandle {registeredTypeIndex=1}, varData []}}
      SamplingRegions []
     }}
    Socket Sampling Mode=::NewEnumerator5
   ) {enabled=false}
   &FN_SetSpawnGroup(Number of Spawn Groups=2)
   &SkeletalMeshLocation(
    Bone / Socket Index=&Add_Integer(A=&SpawnGroup(), B=2)
    Inherit Velocity=false
    Interpolate Mesh Sampling=false
    MaskSpawnGroup=false
    Mesh Sampling Type=::NewEnumerator5
    Sampled Position Offset=<-1.5 1.5 -0.7>
    Sampled Position Offset Coordinate Space:ENiagaraMeshTransforms=::NewEnumerator3
    Skeletal Mesh=$User.SkeletalMesh
    Socket Sampling Mode=::NewEnumerator5
   ) {module=SkeletalMeshLocation001}
   &SphereLocation(
    MaskSpawnGroup=false
    Override Seed=false
    Sphere Radius=1.5
    Surface Only=false
    UseNonUniformScale=false
    UseOffset=false
    UseSphereOrientation=false
   )
  ]
  particleUpdateScript [
   &ParticleState()
   &DynamicMaterialParameters(
    Index 0 Param 3=&FloatFromCurve(
     FloatCurve=DataInterface() {propertyValues {Curve {keys [
         <at 0 0 RCIM_Cubic>
         {arriveTangent=2.4962, interpMode=RCIM_Cubic, leaveTangent=2.4962, tangentMode=RCTM_User, time=0.3, value=0.45}
         <at 1 1>
        ]}}}
     Scale Curve=2
    )
    Module.Param0WriteEnabled=false
    Module.Param1WriteEnabled=false
    Module.Param2WriteEnabled=true
    Module.Param3WriteEnabled=false
   )
   &ScaleMeshSize(Scale Factor=&VectorFromFloat(Value=&FloatFromCurve(
      FloatCurve=DataInterface() {propertyValues {Curve {keys [
          {arriveTangent=-0.026, interpMode=RCIM_Cubic, leaveTangent=-0.026, tangentMode=RCTM_User, time=0, value=0}
          {arriveTangent=0.1004, interpMode=RCIM_Cubic, leaveTangent=0.1004, tangentMode=RCTM_User, time=0.6, value=0.7}
          {arriveTangent=-3.4821, interpMode=RCIM_Cubic, leaveTangent=-3.4821, tangentMode=RCTM_User, time=1, value=0}
         ]}}}
      Scale Curve=0.5
     ))) {module=ScaleMeshSize001}
   &GravityForce(Coordinate Space=::Local, Gravity=<16 -5 0>)
   &Drag(Use Linear Drag=true, Use Rotational Drag=false)
   &MeshRotationForce(
    Coordinate Space:ENiagaraMeshTransforms=::NewEnumerator3
    Rotation=&RandomRangeVector(Minimum=<-4 -4 -4>, Maximum=<4 4 4>)
   )
   &SolveForcesAndVelocity(Clamp Velocity=false, Limit Acceleration=false)
   &SolveRotationalForcesAndVelocity()
   &CameraOffset(Camera Offset Amount=&RandomRangeFloat(Minimum=0, Maximum=-3)) {module=CameraOffset001}
   /SetVariables(Particles.VisibilityTag=&MakeCustomIntFromBool(Bool=$User.bisHeadHidden, True INT=1, False INT=0)) {isSetParameters=true, module=SetVariables_#2}
  ]
  rendererClasses [&NiagaraMeshRendererProperties]
  renderers [
   {rendererClass=&NiagaraMeshRendererProperties, rendererIndex=0}
   {rendererClass=&NiagaraMeshRendererProperties, rendererIndex=1}
  ]
 }
 Light {
  emitterUpdateScript [
   &EmitterState(Inactive Response=::NewEnumerator1, Life Cycle Mode=::NewEnumerator1, Loop Duration=2, UseLoopDelay=true)
   &SpawnBurst_Instantaneous(Spawn Count=2, Use Spawn Probability=false)
  ]
  particleSpawnScript [
   /Niagara/Modules/Spawn/Initialization/InitializeParticle(
    Color=<rgba 0.0361 4 1.6878 1>
    Lifetime=2
    Module.Write Color=true
    Module.Write Lifetime=true
    Module.Write Mass=false
    Module.Write Material Random=true
    Module.Write Position=false
    Module.Write Scale=false
    Module.Write SpriteRotation=false
    Module.Write SpriteSize=true
    Position=&SimulationPosition()
    Sprite Size=<75 75>
   )
   &DynamicMaterialParameters(
    Index 0 Param 1=&UniformRangedFloat(Minimum=0.7, Maximum=0.9)
    Index 0 Param 2=&UniformRangedFloat(Minimum=0.05, Maximum=0.15)
    Index 0 Param 3=1
    Module.Param0WriteEnabled=true
    Module.Param1WriteEnabled=true
    Module.Param2WriteEnabled=false
    Module.Param3WriteEnabled=false
   ) {module=DynamicMaterialParameters002}
   &FN_SetSpawnGroup(Number of Spawn Groups=2)
  ]
  particleUpdateScript [
   &ParticleState()
   &SkeletalMeshLocation(
    Bone / Socket Index=&SpawnGroup()
    Inherit Velocity=false
    Interpolate Mesh Sampling=false
    MaskSpawnGroup=false
    Mesh Sampling Type=::NewEnumerator5
    Sampled Position Offset=&MaskVectorBySpawnGroup(Masked Vector=<1 1.5 0>, Passthrough Vector=<-1 1.5 0>)
    Sampled Position Offset Coordinate Space:ENiagaraMeshTransforms=::NewEnumerator3
    Skeletal Mesh=$User.SkeletalMesh
    Socket Sampling Mode=::NewEnumerator5
   )
   &SolveForcesAndVelocity(Clamp Velocity=false, Limit Acceleration=false)
   &ScaleColor(
    Scale RGB=&Multiply_Vector(
     A=&MakeVectorFromLinearColorRGB(Color=&LerpLinearColors(
       StartColor=<rgba 0.3059 1.9435 2 1>
       EndColor=<rgba 0 0.3333 2 1>
       LerpFactor=&FloatFromCurve(
        FloatCurve=DataInterface() {propertyValues {Curve {keys [<at 0 1>,<at 0.1069 0.0895>,<at 0.3206 0.7423>,<at 0.4907 0.1444>,<at 0.6992 0.8948>,<at 0.9061 0.1261>,<at 1 1>]}}}
        CurveIndex=$Emitter.NormalizedLoopAge
        Scale Curve=0.9
       )
      ))
     B=&VectorFromFloat(Value=&FloatFromCurve(
       FloatCurve=DataInterface() {propertyValues {Curve {keys [<at 0 1>,<at 0.0584 1.1216>,<at 0.0797 0.6691>,<at 0.1464 0.526>,<at 0.222 1.0327>,<at 0.2705 0.6844>,<at 0.3166 0.8798>,<at 0.3641 0.5477>,<at 0.4407 0.6405>,<at 0.4815 1.1358>,<at 0.5388 1.0006>,<at 0.58 0.5314>,<at 0.6479 0.8407>,<at 0.6937 0.7459>,<at 0.737 1.0896>,<at 0.8125 0.5703>,<at 0.8909 0.8732>,<at 0.9449 0.6186>,<at 1 1>]}}}
       CurveIndex=$Emitter.NormalizedLoopAge
      ))
    )
    ScaleA=true
    ScaleRGB=true
   )
   &FN_Light_v2(
    Light Exponent=1
    Light Radius=5
    LightEnabled:NiagaraBool=&Boolean_NotOperation(Bool=$User.bisHeadHidden)
    Scale Color=2
   )
   &ScaleSpriteSize(Scale Factor=&Vector2DFromFloat(Value=&FloatFromCurve(
      FloatCurve=DataInterface() {propertyValues {Curve {keys [<at 0 1>,<at 0.0562 0.9129>,<at 0.0748 1.1321>,<at 0.1235 1.3668>,<at 0.2098 0.8951>,<at 0.2589 1.1544>,<at 0.2987 1.0516>,<at 0.3338 1.2874>,<at 0.3749 1.3981>,<at 0.4815 0.8928>,<at 0.5082 1.113>,<at 0.5446 1.2763>,<at 0.6237 1.1354>,<at 0.6838 1.0572>,<at 0.7109 0.8761>,<at 0.7888 0.7005>,<at 0.8531 1.2103>,<at 0.9284 1.1544>,<at 1 1>]}}}
      CurveIndex=$Emitter.NormalizedLoopAge
     )))
  ]
  rendererClasses [&NiagaraLightRendererProperties]
  renderers [{rendererClass=&NiagaraLightRendererProperties, rendererIndex=0}]
 }
}
summary {
 systemName=NS_M_DarkPit_Eyes
 emitters [
  {emitterName=Light, rendererClasses [&NiagaraLightRendererProperties]}
  {emitterName=Glow, rendererClasses [&NiagaraSpriteRendererProperties]}
  {emitterName=EyeTrail, rendererClasses [&NiagaraMeshRendererProperties]}
  {emitterName=EyeFlamesLeft, rendererClasses [&NiagaraSpriteRendererProperties]}
  {emitterName=EyeFlamesRight, rendererClasses [&NiagaraSpriteRendererProperties]}
  {emitterName=Goop, rendererClasses [&NiagaraMeshRendererProperties]}
  {emitterName=Goop_Big, rendererClasses [&NiagaraMeshRendererProperties]}
 ]
 userVariables [
  User.bisHeadHidden {type=NiagaraBool}
  User.IsSkydiving {type=NiagaraBool}
  User.SkeletalMesh {
   type=NiagaraDataInterfaceSkeletalMesh
   defaultValue {
    struct=NiagaraExt_VariableValue_DataInterface
    value {dataInterface=&self:NiagaraDataInterfaceSkeletalMesh_0}
   }
  }
 ]
}
