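# Physics asset for the Dark Pit medium male player body, covering secondary motion of its belt, hoodie flaps and hoodie zipper.
# Box bodies form a three-link belt chain from a master rotationally locked to the pelvis.
# Capsules form two-link flap chains off spine_05 and a zipper link off its master, all with tight-to-moderate angular limits.
# Pelvis, spine and right-thigh capsules add body shapes; spine, thigh and zipper master rotate freely against the pelvis.
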PhysicsAsset /BRCosmetics/Characters/Player/Male/Medium/Bodies/M_MED_Dark_Pit/Meshes/M_MED_Dark_Pit_Physics
bodies {
 dyn_belt_1_jnt [{
   shapeType=Box
   extentX=0.8018
   extentY=5.5136
   extentZ=3.4334
   shapeName=dyn_belt_1_jnt_box
   center=<2.9728 0.2571 -0.0001>
   rotation=<rot 0 -89.9999 0>
  }]
 dyn_belt_2_jnt [{
   shapeType=Box
   extentX=1.323
   extentY=4.7551
   extentZ=3.6041
   shapeName=dyn_belt_2_jnt_box
   center=<2.705 -0.0001 0.1135>
   rotation=<rot 0 89.9997 0>
  }]
 dyn_belt_3_jnt [{
   shapeType=Box
   extentX=1.0412
   extentY=5.5772
   extentZ=3.1268
   shapeName=dyn_belt_3_jnt_box
   center=<2.7313 -0.122 0>
   rotation=<rot 0 89.9997 0>
  }]
 dyn_belt_master_jnt [{
   shapeType=Box
   extentX=0.8447
   extentY=2.2234
   extentZ=4.0747
   shapeName=dyn_belt_master_jnt_box
   center=<1.1941 0.2315 0>
   rotation=<rot 0 89.9999 0>
  }]
 dyn_hoodie_zipper_jnt [{
   shapeType=Capsule
   length=2.162
   radius=0.7115
   shapeName=dyn_hoodie_zipper_jnt_capsule
   center=<1.456 0.0009 -0.0448>
   rotation=<rot 0 94.4077 -89.3739>
  }]
 dyn_hoodie_zipper_master_jnt [{shapeType=Capsule, length=0.505, radius=0.505, shapeName=dyn_hoodie_zipper_master_jnt_capsule, rotation=<rot 0 -90 0>}]
 dyn_l_hoodieflap_1_jnt [{shapeType=Capsule, length=2, radius=2, shapeName=dyn_l_hoodieflap_1_jnt_capsule, center=<2.6352 -0.0009 0>, rotation=<rot -90 0.0298 -0.03>}]
 dyn_l_hoodieflap_2_jnt [{shapeType=Capsule, length=2, radius=2, shapeName=dyn_l_hoodieflap_2_jnt_capsule, center=<2.6352 -0.0009 0>, rotation=<rot -90 0.0298 -0.03>}]
 dyn_r_hoodieflap_1_jnt [{
   shapeType=Capsule
   length=2
   radius=2
   shapeName=dyn_r_hoodieflap_1_jnt_capsule
   center=<-2.9031 -0.0008 0>
   rotation=<rot -90 0.0199 -0.0201>
  }]
 dyn_r_hoodieflap_2_jnt [{
   shapeType=Capsule
   length=2
   radius=2
   shapeName=dyn_r_hoodieflap_2_jnt_capsule
   center=<-2.9031 -0.0008 0>
   rotation=<rot -90 0.0199 -0.0201>
  }]
 pelvis [{shapeType=Capsule, length=10, radius=10, shapeName=pelvis_capsule}]
 spine_05 [{shapeType=Capsule, length=1, radius=1, shapeName=spine_05_capsule}]
 thigh_twist_01_r [{
   shapeType=Capsule
   length=23.25
   radius=8.25
   shapeName=thigh_twist_01_r_capsule
   center=<-13.0173 0.0013 0.8727>
   rotation=<rot 90 172.875 -177.1247>
  }]
}
constraints [
 {bone1Name=dyn_l_hoodieflap_2_jnt, bone2Name=dyn_l_hoodieflap_1_jnt, swing1LimitDegrees=10, swing2LimitDegrees=10, twistLimitDegrees=2}
 {bone1Name=dyn_r_hoodieflap_2_jnt, bone2Name=dyn_r_hoodieflap_1_jnt, swing1LimitDegrees=10, swing2LimitDegrees=10, twistLimitDegrees=2}
 {bone1Name=dyn_l_hoodieflap_1_jnt, bone2Name=spine_05, swing1LimitDegrees=10, swing2LimitDegrees=10, twistLimitDegrees=2}
 {bone1Name=dyn_r_hoodieflap_1_jnt, bone2Name=spine_05, swing1LimitDegrees=10, swing2LimitDegrees=10, twistLimitDegrees=2}
 {bone1Name=dyn_hoodie_zipper_jnt, bone2Name=dyn_hoodie_zipper_master_jnt, swing1LimitDegrees=45, swing2LimitDegrees=45, twistLimitDegrees=5}
 {bone1Name=dyn_belt_2_jnt, bone2Name=dyn_belt_1_jnt, swing1LimitDegrees=25, swing2LimitDegrees=25, twistLimitDegrees=5}
 {bone1Name=dyn_belt_3_jnt, bone2Name=dyn_belt_2_jnt, swing1LimitDegrees=25, swing2LimitDegrees=25, twistLimitDegrees=5}
 {bone1Name=dyn_belt_1_jnt, bone2Name=dyn_belt_master_jnt, swing1LimitDegrees=25, swing2LimitDegrees=25, twistLimitDegrees=2}
 {
  bone1Name=spine_05
  bone2Name=pelvis
  swing1LimitDegrees=45
  swing1Motion=Free
  swing2LimitDegrees=45
  swing2Motion=Free
  twistLimitDegrees=45
  twistMotion=Free
 }
 {
  bone1Name=dyn_hoodie_zipper_master_jnt
  bone2Name=pelvis
  swing1LimitDegrees=45
  swing1Motion=Free
  swing2LimitDegrees=45
  swing2Motion=Free
  twistLimitDegrees=45
  twistMotion=Free
 }
 {
  bone1Name=thigh_twist_01_r
  bone2Name=pelvis
  swing1LimitDegrees=45
  swing1Motion=Free
  swing2LimitDegrees=45
  swing2Motion=Free
  twistLimitDegrees=45
  twistMotion=Free
 }
 {
  bone1Name=dyn_belt_master_jnt
  bone2Name=pelvis
  swing1LimitDegrees=45
  swing1Motion=Locked
  swing2LimitDegrees=45
  swing2Motion=Locked
  twistLimitDegrees=45
  twistMotion=Locked
 }
]
