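# Physics asset for the shoes of the M_MED_Dark_Pit male player body, covering jointed shoe tongues on both feet.
# Capsule bodies sit on the calves, feet and two dynamic tongue bones per shoe.
# Each tongue forms a two-link chain off its foot with very tight swing and twist limits, allowing only slight wobble.
# Feet link to calves with free angular motion.
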PhysicsAsset /BRCosmetics/Characters/Player/Male/Medium/Bodies/M_MED_Dark_Pit/Meshes/M_MED_Dark_Pit_Physics_Shoes
bodies {
 calf_l [{shapeType=Capsule, length=10, radius=4, shapeName=calf_l_capsule, center=<-31.7019 -1.8879 0.0001>, rotation=<rot -90 -0.633 0.6329>}]
 calf_r [{shapeType=Capsule, length=10, radius=4, shapeName=calf_r_capsule, center=<31.5178 1.6429 -0.0001>, rotation=<rot -90 -13.2627 13.2626>}]
 dyn_L_shoe_tongue_1_jnt [{shapeType=Capsule, length=4.8006, radius=2.1835, shapeName=dyn_L_shoe_tongue_1_jnt_capsule, center=<-1.8315 0 0.0001>}]
 dyn_L_shoe_tongue_2_jnt [{shapeType=Capsule, length=5.9773, radius=3.6788, shapeName=dyn_L_shoe_tongue_2_jnt_capsule, center=<-3.8613 1.5392 0.0002>}]
 dyn_R_shoe_tongue_1_jnt [{shapeType=Capsule, length=4.8006, radius=2.1835, shapeName=dyn_R_shoe_tongue_1_jnt_capsule, center=<1.7936 0 0.0001>}]
 dyn_R_shoe_tongue_2_jnt [{shapeType=Capsule, length=5.9773, radius=3.6788, shapeName=dyn_R_shoe_tongue_2_jnt_capsule, center=<3.84 -1.3281 0.0002>}]
 foot_l [{shapeType=Capsule, length=2, radius=2, shapeName=foot_l_capsule}]
 foot_r [{shapeType=Capsule, length=2, radius=2, shapeName=foot_r_capsule}]
}
constraints [
 {bone1Name=dyn_L_shoe_tongue_1_jnt, bone2Name=foot_l, swing1LimitDegrees=2, swing2LimitDegrees=2, twistLimitDegrees=2}
 {bone1Name=dyn_L_shoe_tongue_2_jnt, bone2Name=dyn_L_shoe_tongue_1_jnt, swing1LimitDegrees=2, swing2LimitDegrees=2, twistLimitDegrees=2}
 {bone1Name=dyn_R_shoe_tongue_1_jnt, bone2Name=foot_r, swing1LimitDegrees=2, swing2LimitDegrees=2, twistLimitDegrees=2}
 {bone1Name=dyn_R_shoe_tongue_2_jnt, bone2Name=dyn_R_shoe_tongue_1_jnt, swing1LimitDegrees=2, swing2LimitDegrees=2, twistLimitDegrees=2}
 {
  bone1Name=foot_r
  bone2Name=calf_r
  swing1LimitDegrees=45
  swing1Motion=Free
  swing2LimitDegrees=45
  swing2Motion=Free
  twistLimitDegrees=45
  twistMotion=Free
 }
 {
  bone1Name=foot_l
  bone2Name=calf_l
  swing1LimitDegrees=45
  swing1Motion=Free
  swing2LimitDegrees=45
  swing2Motion=Free
  twistLimitDegrees=45
  twistMotion=Free
 }
]
