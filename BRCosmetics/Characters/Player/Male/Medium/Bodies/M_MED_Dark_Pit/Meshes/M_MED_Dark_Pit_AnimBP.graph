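# Animation Blueprint for the Dark Pit male medium body cosmetic, copying the attached parent mesh's pose and layering deformation and secondary motion on top.
# Twist and sculpter Control Rigs correct deformation, then two RigidBody nodes simulate the main and shoe physics assets relative to root and pelvis.
# Additive bone offsets then adjust the back collar while skydiving and the right hoodie flap while targeting.
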AnimBlueprint /BRCosmetics/Characters/Player/Male/Medium/Bodies/M_MED_Dark_Pit/Meshes/M_MED_Dark_Pit_AnimBP
target_skeleton=&/M_MED_Dark_Pit_Skeleton
graphs [
 AnimGraph {nodes [
   n1 = Control Rig(Source=Copy Pose From Mesh() {bUseAttachedParent=true}) {
    alphaBoolBlend {blendOption=Linear}
    controlRigAssetReference {blueprintRigClass=&/M_MED_Dark_Pit_Deform_CtrlRig_twists.M_MED_Dark_Pit_Deform_CtrlRig_twists_C}
   }
   n2 = Control Rig!(
    Source=@n1
    Alpha
    KneeRot_fr_l_114deg:Vector
    KneeTrans_fr_l_114deg:Vector="0.258,-3.591,0.009"
    KneeScale_fr_l_114deg:Vector="1,1,1"
    KneeScale_fr_l_130deg:Vector="1,1,1"
    KneeRot_fr_l_130deg:Vector="0,0,-51.2"
    KneeTrans_fr_l_130deg:Vector="0.258,-3.591,0.009"
   ) -> (Pose) {
    alphaBoolBlend {blendOption=Linear}
    controlRigAssetReference {blueprintRigClass=&/M_MED_Dark_Pit_Deform_CtrlRig_sculpters.M_MED_Dark_Pit_Deform_CtrlRig_sculpters_C}
   }
   n3 = RigidBody(ComponentPose=Local To Component(LocalPose=@n2)) {
    bClampLinearTranslationLimitToRefPose=true
    bOverrideWorldGravity=true
    overridePhysicsAsset=&/M_MED_Dark_Pit_Physics
    simulationSpace=BaseBoneSpace
    alphaBoolBlend {blendOption=Linear}
    baseBoneRef {boneName=root}
    componentAppliedLinearAccClamp=<3000 3000 3000>
    componentLinearAccScale=<0.3 0.2 0.2>
    componentLinearVelScale=<0.3 0.2 0.2>
    overrideWorldGravity {z=-980}
   }
   n4 = RigidBody(ComponentPose=@n3) {
    bClampLinearTranslationLimitToRefPose=true
    bOverrideWorldGravity=true
    overridePhysicsAsset=&/M_MED_Dark_Pit_Physics_Shoes
    simulationSpace=BaseBoneSpace
    alphaBoolBlend {blendOption=Linear}
    baseBoneRef {boneName=pelvis}
    componentAppliedLinearAccClamp=<3000 3000 3000>
    componentLinearAccScale=<0.1 0.05 0.05>
    componentLinearVelScale=<0.1 0.05 0.05>
    overrideWorldGravity {z=-980}
   }
   n5 = "Transform (Modify) Bone - Bone: dyn_collar_back_jnt"(ComponentPose=@n4, Alpha=SelectFloat(A=1, B=0, bPickA=Get bIsSkyDiving()), Translation="0,0,0", Rotation="0,0,-15", Scale="1,1,1") {
    rotationMode=BMM_Additive
    rotationSpace=BCS_BoneSpace
    alphaBoolBlend {blendOption=Linear}
    boneToModify {boneName=dyn_collar_back_jnt}
    rotation {roll=-15}
   }
   n6 = "Transform (Modify) Bone - Bone: dyn_r_hoodieflap_1_jnt"(ComponentPose=@n5, Alpha=SelectFloat(A=1, B=0, bPickA=Get bIsTargeting()), Translation="0,1,-3", Rotation="0,0,0", Scale="1,1,1") {
    rotationSpace=BCS_WorldSpace
    translationMode=BMM_Additive
    translationSpace=BCS_BoneSpace
    alphaBoolBlend {blendOption=Linear}
    boneToModify {boneName=dyn_r_hoodieflap_1_jnt}
    translation {y=1, z=-3}
   }
   Output Pose(Result=Component To Local(ComponentPose=@n6)) {layerGroup=None}
  ]}
 EventGraph {links [], nodes []}
]
variables [Skydiving {type=number, default=1}]
