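# Physics asset for the face accessory part of the medium male Dark Pit cosmetic body, covering the bones of a hood.
# Five single-capsule bodies sit on a master hood bone and its back, left, right and top-front joints.
# Each outer joint is constrained directly to the master in a star layout, with swing and twist limits.
# The back is held tightest, the top-front moves moderately and the sides swing most freely.
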PhysicsAsset /BRCosmetics/Characters/Player/Male/Medium/Bodies/M_MED_Dark_Pit/Meshes/Parts/M_MED_Dark_Pit_FaceAcc_Physics
bodies {
 dyn_back_hood_jnt [{shapeType=Capsule, length=2.75, radius=1.75, shapeName=dyn_back_hood_jnt_capsule, center=<2.2063 0.0002 0>}]
 dyn_hood_master [{shapeType=Capsule, length=2, radius=2, shapeName=dyn_hood_master_capsule}]
 dyn_l_hood_jnt [{shapeType=Capsule, length=2, radius=2, shapeName=dyn_l_hood_jnt_capsule}]
 dyn_r_hood_jnt [{shapeType=Capsule, length=2, radius=2, shapeName=dyn_r_hood_jnt_capsule}]
 dyn_topfront_hood_jnt [{
   shapeType=Capsule
   length=1.75
   radius=3.25
   shapeName=dyn_topfront_hood_jnt_capsule
   center=<0.2694 0 0.0001>
   rotation=<rot 0.0003 -30.0001 90.0005>
  }]
}
constraints [
 {bone1Name=dyn_back_hood_jnt, bone2Name=dyn_hood_master, swing1LimitDegrees=5, swing2LimitDegrees=10, twistLimitDegrees=1}
 {bone1Name=dyn_r_hood_jnt, bone2Name=dyn_hood_master, swing1LimitDegrees=45, swing2LimitDegrees=45, twistLimitDegrees=5}
 {bone1Name=dyn_l_hood_jnt, bone2Name=dyn_hood_master, swing1LimitDegrees=45, swing2LimitDegrees=45, twistLimitDegrees=5}
 {bone1Name=dyn_topfront_hood_jnt, bone2Name=dyn_hood_master, swing1LimitDegrees=20, swing2LimitDegrees=20, twistLimitDegrees=5}
]
